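# Physics asset for the cape on the Nautical Pajamas backpack cosmetic, using capsule bodies along its cape bones.
# Three four-link chains (center, mid-left, mid-right) hang from MSTR anchor bodies with free angular motion to backpack_root, using limited swing and twist between links.
# Cross constraints tie each side link to its center counterpart, forming a lattice that holds the cape together.
# Spine, clavicle and arm capsules plus a floor box supply collision shapes.
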PhysicsAsset /BRCosmetics/Accessories/FORT_Backpacks/Backpack_M_MED_Nautical_Pajamas/Meshes/M_MED_Nautical_Pajamas_Pack_Cape_Capsule_Physics
bodies {
 backpack_root [{shapeType=Capsule, length=1, radius=1}]
 clavicle_l [{shapeType=Capsule, length=15, radius=8.755, center=<6.9 0 0>, rotation=<rot 90 0 0>}]
 clavicle_r [{shapeType=Capsule, length=15, radius=8.755, center=<-6.9 0 0>, rotation=<rot 90 0 0>}]
 dyn_cape_01 [{shapeType=Capsule, length=4.5, radius=4, center=<4.9361 0 0>, rotation=<rot 90 0 0>}]
 dyn_cape_01_MSTR [{shapeType=Capsule, length=1, radius=1, rotation=<rot 1 1 1>}]
 dyn_cape_01_mid_l [{shapeType=Capsule, length=4.5, radius=4, center=<4.9361 0 0>, rotation=<rot 90 0 0>}]
 dyn_cape_01_mid_l_MSTR [{shapeType=Capsule, length=1, radius=1, rotation=<rot 1 1 1>}]
 dyn_cape_01_mid_r [{shapeType=Capsule, length=4.5, radius=4, center=<4.9361 0 0>, rotation=<rot 90 0 0>}]
 dyn_cape_01_mid_r_MSTR [{shapeType=Capsule, length=1, radius=1, rotation=<rot 1 1 1>}]
 dyn_cape_02 [{shapeType=Capsule, length=4.5, radius=4, center=<4.9361 0 0>, rotation=<rot 90 0 0>}]
 dyn_cape_02_mid_l [{shapeType=Capsule, length=4.5, radius=4, center=<4.9361 0 0>, rotation=<rot 90 0 0>}]
 dyn_cape_02_mid_r [{shapeType=Capsule, length=4.5, radius=4, center=<4.9361 0 0>, rotation=<rot 90 0 0>}]
 dyn_cape_03 [{shapeType=Capsule, length=4.5, radius=4, center=<4.9361 0 0>, rotation=<rot 90 0 0>}]
 dyn_cape_03_mid_l [{shapeType=Capsule, length=4.5, radius=4, center=<4.9361 0 0>, rotation=<rot 90 0 0>}]
 dyn_cape_03_mid_r [{shapeType=Capsule, length=4.5, radius=4, center=<4.9361 0 0>, rotation=<rot 90 0 0>}]
 dyn_cape_04 [{shapeType=Capsule, length=4.5, radius=4, center=<4.9361 0 0>, rotation=<rot 90 0 0>}]
 dyn_cape_04_mid_l [{shapeType=Capsule, length=4.5, radius=4, center=<4.9361 0 0>, rotation=<rot 90 0 0>}]
 dyn_cape_04_mid_r [{shapeType=Capsule, length=4.5, radius=4, center=<4.9361 0 0>, rotation=<rot 90 0 0>}]
 floorCOL [{shapeType=Box, extentX=226.01, extentY=202.51, extentZ=3.51}]
 lowerarm_l [{shapeType=Capsule, length=17.75, radius=6.005, center=<7 0 0>, rotation=<rot 90 0 0>}]
 lowerarm_r [{shapeType=Capsule, length=17.75, radius=6.005, center=<-6.9 0 0>, rotation=<rot 90 0 0>}]
 spine_01 [{shapeType=Capsule, length=11.5, radius=13.5, rotation=<rot 0 -90 0>}]
 spine_03 [{shapeType=Capsule, length=11.75, radius=13.25, rotation=<rot 0 -90 0>}]
 spine_05 [{shapeType=Capsule, length=12.25, radius=12.75, rotation=<rot 0 -90 0>}]
 upperarm_l [{shapeType=Capsule, length=16.75, radius=7.005, center=<7 0 0>, rotation=<rot 90 0 0>}]
 upperarm_r [{shapeType=Capsule, length=16.75, radius=7.005, center=<-7 0 0>, rotation=<rot 90 0 0>}]
}
constraints [
 {bone1Name=dyn_cape_01, bone2Name=dyn_cape_01_MSTR, swing1LimitDegrees=10, swing2LimitDegrees=5, twistLimitDegrees=5}
 {bone1Name=dyn_cape_02, bone2Name=dyn_cape_01, swing1LimitDegrees=45, swing2LimitDegrees=45, twistLimitDegrees=15}
 {bone1Name=dyn_cape_03, bone2Name=dyn_cape_02, swing1LimitDegrees=45, swing2LimitDegrees=45, twistLimitDegrees=15}
 {bone1Name=dyn_cape_04, bone2Name=dyn_cape_03, swing1LimitDegrees=45, swing2LimitDegrees=45, twistLimitDegrees=15}
 {
  bone1Name=dyn_cape_01_mid_l_MSTR
  bone2Name=backpack_root
  swing1LimitDegrees=45
  swing1Motion=Free
  swing2LimitDegrees=45
  swing2Motion=Free
  twistLimitDegrees=45
  twistMotion=Free
 }
 {
  bone1Name=dyn_cape_01_mid_r_MSTR
  bone2Name=backpack_root
  swing1LimitDegrees=45
  swing1Motion=Free
  swing2LimitDegrees=45
  swing2Motion=Free
  twistLimitDegrees=45
  twistMotion=Free
 }
 {bone1Name=dyn_cape_01_mid_r, bone2Name=dyn_cape_01_mid_r_MSTR, swing1LimitDegrees=10, swing2LimitDegrees=5, twistLimitDegrees=5}
 {
  bone1Name=dyn_cape_01_MSTR
  bone2Name=backpack_root
  swing1LimitDegrees=45
  swing1Motion=Free
  swing2LimitDegrees=45
  swing2Motion=Free
  twistLimitDegrees=45
  twistMotion=Free
 }
 {bone1Name=dyn_cape_02_mid_r, bone2Name=dyn_cape_01_mid_r, swing1LimitDegrees=45, swing2LimitDegrees=45, twistLimitDegrees=15}
 {bone1Name=dyn_cape_03_mid_r, bone2Name=dyn_cape_02_mid_r, swing1LimitDegrees=45, swing2LimitDegrees=45, twistLimitDegrees=15}
 {bone1Name=dyn_cape_04_mid_r, bone2Name=dyn_cape_03_mid_r, swing1LimitDegrees=5, swing2LimitDegrees=5, twistLimitDegrees=5}
 {bone1Name=dyn_cape_01_mid_l, bone2Name=dyn_cape_01_mid_l_MSTR, swing1LimitDegrees=10, swing2LimitDegrees=5, twistLimitDegrees=5}
 {bone1Name=dyn_cape_02_mid_l, bone2Name=dyn_cape_01_mid_l, swing1LimitDegrees=45, swing2LimitDegrees=45, twistLimitDegrees=15}
 {bone1Name=dyn_cape_03_mid_l, bone2Name=dyn_cape_02_mid_l, swing1LimitDegrees=45, swing2LimitDegrees=45, twistLimitDegrees=15}
 {bone1Name=dyn_cape_04_mid_l, bone2Name=dyn_cape_03_mid_l, swing1LimitDegrees=5, swing2LimitDegrees=5, twistLimitDegrees=5}
 {bone1Name=dyn_cape_01_mid_r, bone2Name=dyn_cape_01, swing1LimitDegrees=20, swing2LimitDegrees=45, twistLimitDegrees=20}
 {bone1Name=dyn_cape_01_mid_l, bone2Name=dyn_cape_01, swing1LimitDegrees=20, swing2LimitDegrees=45, twistLimitDegrees=20}
 {bone1Name=dyn_cape_02_mid_r, bone2Name=dyn_cape_02, swing1LimitDegrees=20, swing2LimitDegrees=45, twistLimitDegrees=20}
 {bone1Name=dyn_cape_02_mid_l, bone2Name=dyn_cape_02, swing1LimitDegrees=20, swing2LimitDegrees=45, twistLimitDegrees=20}
 {bone1Name=dyn_cape_03_mid_r, bone2Name=dyn_cape_03, swing1LimitDegrees=20, swing2LimitDegrees=45, twistLimitDegrees=20}
 {bone1Name=dyn_cape_03_mid_l, bone2Name=dyn_cape_03, swing1LimitDegrees=20, swing2LimitDegrees=45, twistLimitDegrees=20}
 {bone1Name=dyn_cape_04_mid_r, bone2Name=dyn_cape_04, swing1LimitDegrees=20, swing2LimitDegrees=45, twistLimitDegrees=20}
 {bone1Name=dyn_cape_04_mid_l, bone2Name=dyn_cape_04, swing1LimitDegrees=20, swing2LimitDegrees=45, twistLimitDegrees=20}
]
