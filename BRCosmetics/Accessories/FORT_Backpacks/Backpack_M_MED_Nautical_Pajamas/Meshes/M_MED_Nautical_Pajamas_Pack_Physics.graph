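# Physics asset for the Nautical Pajamas backpack cosmetic, defining collision and joint limits for its wing, eye and pillow bones.
# Each bone carries a capsule body, with the backpack root as the shared anchor.
# Mirrored left and right three-bone wing and eye chains hang from the root, while each side pillow branches into top and bottom pieces.
# Tight swing and twist limits keep motion small, with some left-right asymmetry.
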PhysicsAsset /BRCosmetics/Accessories/FORT_Backpacks/Backpack_M_MED_Nautical_Pajamas/Meshes/M_MED_Nautical_Pajamas_Pack_Physics
bodies {
 backpack_root [{shapeType=Capsule, length=9.1587, radius=3.5451, center=<0.0574 -3.0695 0.6723>, rotation=<rot 0 -89.8572 -89.5402>}]
 eye_01_l [{shapeType=Capsule, length=0.35, radius=1.25}]
 eye_01_r [{shapeType=Capsule, length=0.35, radius=1.25}]
 eye_02_l [{shapeType=Capsule, length=0.35, radius=1.25}]
 eye_02_r [{shapeType=Capsule, length=0.35, radius=1.25}]
 eye_03_l [{shapeType=Capsule, length=0.35, radius=1.25}]
 eye_03_r [{shapeType=Capsule, length=0.35, radius=1.25}]
 pillow_bl [{shapeType=Capsule, length=2.2464, radius=1.5784, center=<0.5975 0 -0.1961>, rotation=<rot -18.1669 0 0>}]
 pillow_br [{shapeType=Capsule, length=2.1141, radius=1.5469, rotation=<rot 44.0502 0 0>}]
 pillow_l [{shapeType=Capsule, length=3.7787, radius=2.3618, center=<1.8732 -0.6728 -0.0294>, rotation=<rot 14.383 -90 -2.3442>}]
 pillow_r [{shapeType=Capsule, length=4.8087, radius=2.2347}]
 pillow_tl [{shapeType=Capsule, length=2.2395, radius=2.0862, center=<0.3044 0 0.3694>, rotation=<rot 21.1836 0 0>}]
 pillow_tr [{shapeType=Capsule, length=2.518, radius=1.5945, rotation=<rot -36.8077 0 0>}]
 wing_01_l [{shapeType=Capsule, length=1.6351, radius=1.9395}]
 wing_01_r [{shapeType=Capsule, length=1.6117, radius=1.6715}]
 wing_02_l [{shapeType=Capsule, length=1.5962, radius=1.25}]
 wing_02_r [{shapeType=Capsule, length=1.3462, radius=1.25}]
 wing_03_l [{shapeType=Capsule, length=1.981, radius=1.25}]
 wing_03_r [{shapeType=Capsule, length=2.231, radius=1.25}]
}
constraints [
 {bone1Name=wing_01_r, bone2Name=backpack_root, swing1LimitDegrees=10, swing2LimitDegrees=10, twistLimitDegrees=5}
 {bone1Name=wing_02_r, bone2Name=wing_01_r, swing1LimitDegrees=10, swing2LimitDegrees=10, twistLimitDegrees=5}
 {bone1Name=wing_03_r, bone2Name=wing_02_r, swing1LimitDegrees=10, swing2LimitDegrees=10, twistLimitDegrees=5}
 {bone1Name=wing_01_l, bone2Name=backpack_root, swing1LimitDegrees=10, swing2LimitDegrees=10, twistLimitDegrees=5}
 {bone1Name=wing_02_l, bone2Name=wing_01_l, swing1LimitDegrees=10, swing2LimitDegrees=10, twistLimitDegrees=5}
 {bone1Name=wing_03_l, bone2Name=wing_02_l, swing1LimitDegrees=10, swing2LimitDegrees=10, twistLimitDegrees=5}
 {bone1Name=eye_01_r, bone2Name=backpack_root, swing1LimitDegrees=4, swing2LimitDegrees=4, twistLimitDegrees=4}
 {bone1Name=eye_02_r, bone2Name=eye_01_r, swing1LimitDegrees=5, swing2LimitDegrees=5, twistLimitDegrees=5}
 {bone1Name=eye_01_l, bone2Name=backpack_root, swing1LimitDegrees=4, swing2LimitDegrees=4, twistLimitDegrees=4}
 {bone1Name=eye_02_l, bone2Name=eye_01_l, swing1LimitDegrees=7, swing2LimitDegrees=7, twistLimitDegrees=7}
 {bone1Name=pillow_l, bone2Name=backpack_root, swing1LimitDegrees=15, swing2LimitDegrees=15, twistLimitDegrees=5}
 {bone1Name=pillow_tl, bone2Name=pillow_l, swing1LimitDegrees=15, swing2LimitDegrees=15, twistLimitDegrees=5}
 {bone1Name=pillow_bl, bone2Name=pillow_l, swing1LimitDegrees=15, swing2LimitDegrees=15, twistLimitDegrees=5}
 {bone1Name=pillow_r, bone2Name=backpack_root, swing1LimitDegrees=5, swing2LimitDegrees=5, twistLimitDegrees=5}
 {bone1Name=pillow_br, bone2Name=pillow_r, swing1LimitDegrees=10, swing2LimitDegrees=10, twistLimitDegrees=5}
 {bone1Name=pillow_tr, bone2Name=pillow_r, swing1LimitDegrees=10, swing2LimitDegrees=10, twistLimitDegrees=5}
 {bone1Name=eye_03_l, bone2Name=eye_02_l, swing1LimitDegrees=8, swing2LimitDegrees=8, twistLimitDegrees=8}
 {bone1Name=eye_03_r, bone2Name=eye_02_r, swing1LimitDegrees=5, swing2LimitDegrees=5, twistLimitDegrees=5}
]
